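# Looping foley SoundCue for the MistMylar emote, playing its traversal loop wave.
# A continuous modulator drives the looping wave player's pitch and volume from the EmoteMoveSpeed parameter.
# It uses emote attenuation, a foley sound class and property sheet, overridden concurrency with a small voice cap, and play-when-silent virtualization.
# It also sends to a music audio-analysis submix.
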SoundCue /BRCosmetics/Sounds/Emotes/MistMylar/EmoteFoley_MistMylar_LP_Cue
root=ModulatorContinuous_0
nodes [
 {
  id=ModulatorContinuous_0
  children [WavePlayer_1]
  properties {pitchModulationParams {minOutput=0.98, parameterName=EmoteMoveSpeed}, volumeModulationParams {minOutput=0.8, parameterName=EmoteMoveSpeed}}
 }
 {id=WavePlayer_1, properties {bLooping=true, soundWaveAssetPtr=&/Emote_MistMylar_Trav_LP}}
]
properties {
 attenuationSettings=&Fort_Emotes_Athena_Att
 audioPropertiesSheet=/BRCosmetics/Sounds/Settings/PropertySheets/Emotes/APS_Cosmetic_Emote_Foley
 bOverrideConcurrency=true
 maxDistance=1600
 soundClassObject=/Game/Sounds/FortSoundClassesAndModes/Fort_Foley_Emote
 virtualizationMode=PlayWhenSilent
 volumeMultiplier=1
 concurrencyOverrides {maxCount=3}
 soundSubmixSends [{minSendLevel=0, soundSubmix=&Fortnite_AudioAnalysis_Music}]
}
